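# Template_Green is a flat green material in the weapon skeleton templates folder, apparently a placeholder color for template meshes.
# Its graph holds a single constant green color vector, with no textures, parameters or math.
# It uses Default Lit shading and is flagged for use on skeletal meshes and with mesh deformers.
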Material /Game/Weapons/Skeleton_Templates/Template_Green
graph {
 properties {bUsedWithMeshDeformer=true, bUsedWithSkeletalMesh=true, pixelDepthOffsetMode=PDOM_Legacy, usedShadingModels=Default Lit}
 nodes [Constant3Vector() {constant=<rgba 0 0.66 0.1502 1>}]
}
